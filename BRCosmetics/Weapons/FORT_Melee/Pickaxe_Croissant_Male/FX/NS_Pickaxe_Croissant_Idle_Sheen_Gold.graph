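# Niagara system for the gold idle sheen on the Croissant male pickaxe cosmetic, built from one emitter with mesh and sprite renderers.
# Two bursts share the emitter, with spawn group written to VisibilityTag: the sheen, then a sparkle timed by an offset before sheen end.
# Group masks set each burst's color, lifetime, position and camera offset.
# A lifetime curve drives a dynamic material parameter, while sprites rotate and pop in before shrinking.
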NiagaraSystem /BRCosmetics/Weapons/FORT_Melee/Pickaxe_Croissant_Male/FX/NS_Pickaxe_Croissant_Idle_Sheen_Gold
dependencies {
 usedDataInterfaces [&NiagaraDataInterfaceCurve, &NiagaraDataInterfaceArrayFloat]
 usedDynamicInputs [&SelectFloatFromArray, &ReturnExecIndex]
 usedModules [
  &SystemState
  &self:SystemSpawnScript.SystemScriptSource.SystemScriptGraph.NiagaraNodeAssignment_0.SetVariables_#1
 ]
}
emitters {Sheen_Mesh_Sparkle {
  emitterUpdateScript [
   &EmitterState()
   &SpawnBurst_Instantaneous(Use Spawn Probability=false, UseLoopCountLimit=false)
   &SpawnBurst_Instantaneous(
    Spawn Group=1
    Spawn Time=&Subtract_Float(A=$System.SheenLifetime, B=$System.SparkleSpawnTimeOffset)
    Use Spawn Probability=false
    UseLoopCountLimit=false
   ) {module=SpawnBurst_Instantaneous001}
  ]
  particleSpawnScript [
   /SetVariables(Particles.VisibilityTag=&SpawnGroup()) {isSetParameters=true}
   &InitializeParticle(
    Color=&MaskLinearColorBySpawnGroup(Masked Linear Color=<rgba 1 0.43 0.1 1>, Passthrough Linear Color=<rgba 10 4.0183 0.3 1>)
    Lifetime=&MaskFloatBySpawnGroup(Masked Float=$System.SheenLifetime, Passthrough Float=&RandomRangeFloat(Minimum=0.666, Maximum=0.777))
    Mesh Scale Mode=::NewEnumerator3
    Position Offset=&MaskVectorBySpawnGroup(Masked Vector=<0 0 0>, Passthrough Vector=$System.SparklePositionOffset)
    Sprite Size=&Multiply_Vector2D_ByFloat(Float=&RandomRangeFloat(Minimum=16, Maximum=23), Vector2D=<1.1 1>)
    Sprite Size Mode=::NewEnumerator4
    UsePositionOffset=true
    Write Material Random=true
   )
   &CameraOffset(Camera Offset Amount=&MaskFloatBySpawnGroup(Masked Float=0, Passthrough Float=1))
  ]
  particleUpdateScript [
   &ParticleState()
   &SpriteRotationRate(Rotation Rate=&RandomRangeFloat(Minimum=90, Maximum=360))
   &ScaleSpriteSize(Uniform Curve Sprite Scale=DataInterface() {propertyValues {Curve {keys [
        {arriveTangent=9.0725, interpMode=RCIM_Cubic, leaveTangent=9.0725, tangentMode=RCTM_User, time=0, value=0}
        {arriveTangent=-1.7917, interpMode=RCIM_Cubic, leaveTangent=-1.7917, tangentMode=RCTM_User, time=0.0666, value=0.9847}
        {arriveTangent=-0.0405, interpMode=RCIM_Cubic, leaveTangent=-0.0405, tangentMode=RCTM_User, time=1, value=0}
       ]}}})
   &DynamicMaterialParameters(
    Index 0 Param 1=&FloatFromCurve(FloatCurve=DataInterface() {propertyValues {Curve {keys [{interpMode=RCIM_Cubic, tangentMode=RCTM_User, time=0, value=0}, {interpMode=RCIM_Cubic, tangentMode=RCTM_User, time=1, value=1}]}}})
    Index 0 Param 2=0.5
    Index 0 Param 3=-0.02
    Module.Index0Param0Write=true
    Module.Index0Param1Write=true
    Module.Index0Param2Write=true
    Module.Index0Param3Write=false
   )
  ]
  rendererClasses [&NiagaraMeshRendererProperties, &NiagaraSpriteRendererProperties]
  renderers [
   {rendererClass=&NiagaraMeshRendererProperties, rendererIndex=0}
   {rendererClass=&NiagaraSpriteRendererProperties, rendererIndex=1}
  ]
 }}
summary {
 systemName=NS_Pickaxe_Croissant_Idle_Sheen_Gold
 emitters [{
   emitterName=Sheen_Mesh_Sparkle
   rendererClasses [&NiagaraMeshRendererProperties, &NiagaraSpriteRendererProperties]
  }]
 userVariables []
}
